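# Niagara impact burst effect for the Gold variant of the DollEnthusiast pickaxe, in a gold-orange palette.
# Burst emitters layer a core flash, flare, glow, speed lines and toon smoke over flung, tumbling mesh shapes with sparkle sprites.
# Toon sparks spawn in bursts and also trail the mesh shapes through a particle attribute reader.
# Front-end modifiers and time-of-day scalars adjust size, speed, camera offset and brightness.
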NiagaraSystem /BRCosmetics/Weapons/FORT_Melee/Pickaxe_DollEnthusiast/FX/Gold/NS_Pickaxe_DollEnthusiast_Gold_Impact
dependencies {
 usedDataInterfaces [
  &NiagaraDataInterfaceCurve
  &NiagaraDataInterfaceArrayColor
  &NiagaraDataInterfaceMeshRendererInfo
  &NiagaraDataInterfaceColorCurve
  &NiagaraDataInterfaceParticleRead
 ]
 usedDynamicInputs [&MakeFloatFromInt, &SimulationPosition]
 usedModules [
  &self:SystemSpawnScript.SystemScriptSource.SystemScriptGraph.NiagaraNodeAssignment_1.SetVariables_#1
  &SystemState
  &self:SystemSpawnScript.SystemScriptSource.SystemScriptGraph.NiagaraNodeAssignment_0.SetVariables_#2
 ]
}
emitters {
 CoreFlashImpact001 {
  emitterUpdateScript [
   &EmitterState()
   &SpawnBurst_Instantaneous(Use Spawn Probability=false, UseLoopCountLimit=false)
  ]
  particleSpawnScript [
   &InitializeParticle(
    Lifetime Max=0.2
    Lifetime Min=0.15
    Lifetime Mode=::NewEnumerator1
    Sprite Rotation Mode=::NewEnumerator2
    Sprite Size Mode=::NewEnumerator3
    Uniform Sprite Size=&Multiply_Float(
     A=&RandomRangeFloat(Minimum=60, Maximum=80)
     B=&FN_FrontEndFloatModifier(Gameplay=1, Front End=$User.FrontEndScale)
    )
    UsePositionOffset=false
    Write Material Random=true
   )
   &FN_PerParticleLOD(Enable SpriteSize Distance Scaling=true, Sprite Size Scale Far=<2.5 2.5>)
   &CameraOffset(Camera Offset Amount=&FN_FrontEndFloatModifier(Gameplay=15, Front End=10))
   /SetVariables(Particles.SubImageIndex=&RandomRangeFloat(Minimum=0, Maximum=4, Randomness Mode=::NewEnumerator2, Random Seed=?)) {isSetParameters=true}
   /SetVariables(Particles.SubImageIndexAlt:NiagaraFloat=&ModuloFloat(
     A=&Add_Float(A=$Particles.SubImageIndex, B=&RandomRangeFloat(Minimum=1, Maximum=2))
     B=3
    )) {isSetParameters=true}
  ]
  particleUpdateScript [
   &ParticleState()
   &ScaleSpriteSize(Uniform Curve Sprite Scale=DataInterface() {propertyValues {Curve {keys [{arriveTangent=2.6565, interpMode=RCIM_Cubic, leaveTangent=2.6565, tangentMode=RCTM_User, time=0, value=0}, <at 1 1 RCIM_Cubic>]}}}) {module=ScaleSpriteSize001}
   &ScaleColor(
    Scale Alpha=&Multiply_Float(
     A=&FloatFromCurve(FloatCurve=DataInterface() {propertyValues {Curve {keys [<at 0 1>, {arriveTangent=-1, interpMode=RCIM_Cubic, leaveTangent=-1, time=0.4, value=1}, <at 1 0 RCIM_Cubic>]}}})
     B=1
    )
    Scale RGB=&MakeVectorFromLinearColorRGB(Color=<rgba 5 5 5 1>)
    ScaleA=true
    ScaleRGB=true
   )
   /SetVariables(
    Particles.Color1Size:Vector2f=&Multiply_Vector2D_ByFloat(Float=1.25, Vector2D=$Particles.SpriteSize)
    Particles.Color2Size:Vector2f=&Multiply_Vector2D_ByFloat(Float=1.9, Vector2D=$Particles.SpriteSize)
    Particles.OutlineSize:Vector2f=&Multiply_Vector2D_ByFloat(Float=2.5, Vector2D=$Particles.SpriteSize)
   ) {isSetParameters=true}
   /SetVariables(
    Particles.Color1:LinearColor=&MultiplyLinearColor(
     Linear Color=<rgba 0 0 0 1>
     Scale Factor=&MakeVector4_FromFloat(Float=&FNTimeOfDayScalar(Night Scale=0.75, Morning Scale=0.9, Evening Scale=0.9))
    )
    Particles.Color2=&MultiplyLinearColor(
     Linear Color=&ColorFromCurve(ColorCurve=DataInterface() {propertyValues {
        AlphaCurve {keys [<at 0 0>,<at 0.1287 1 RCIM_Cubic>,<at 1 0>]}
        BlueCurve {keys [<at 0.004 0>, {arriveTangent=0.077, interpMode=RCIM_Cubic, leaveTangent=0.077, time=0.1965, value=0}, <at 1 0.0767 RCIM_Cubic>]}
        GreenCurve {keys [<at 0.004 0.3994>, {arriveTangent=0.0462, interpMode=RCIM_Cubic, leaveTangent=0.0462, time=0.1965, value=0.8274}, <at 1 0.4454 RCIM_Cubic>]}
        RedCurve {keys [<at 0.004 1>,<at 0.1965 2 RCIM_Cubic>,<at 1 1 RCIM_Cubic>]}
       }})
     Scale Factor=&MakeVector4_FromFloat(Float=&FNTimeOfDayScalar(Night Scale=1, Morning Scale=0.9, Evening Scale=0.9))
    )
    Particles.OutlineColor:LinearColor=&ColorFromCurve(ColorCurve=DataInterface() {propertyValues {
       AlphaCurve {keys [<at 0 0>,<at 0.1287 1 RCIM_Cubic>,<at 1 0>]}
       BlueCurve {keys [<at 0.004 0>, {arriveTangent=0.077, interpMode=RCIM_Cubic, leaveTangent=0.077, time=0.1965, value=0}, <at 1 0.0767 RCIM_Cubic>]}
       GreenCurve {keys [<at 0.004 0.3994>, {arriveTangent=0.0462, interpMode=RCIM_Cubic, leaveTangent=0.0462, time=0.1965, value=0.8274}, <at 1 0.4454 RCIM_Cubic>]}
       RedCurve {keys [<at 0.004 1>,<at 0.1965 2 RCIM_Cubic>,<at 1 1 RCIM_Cubic>]}
      }})
   ) {isSetParameters=true}
  ]
  rendererClasses [&NiagaraSpriteRendererProperties]
  renderers [
   {rendererClass=&NiagaraSpriteRendererProperties}
   {rendererClass=&NiagaraSpriteRendererProperties, rendererIndex=1}
   {rendererClass=&NiagaraSpriteRendererProperties, rendererIndex=2}
   {rendererClass=&NiagaraSpriteRendererProperties, rendererIndex=3}
  ]
 }
 Dark_Glow {
  emitterUpdateScript [
   &EmitterState()
   &SpawnBurst_Instantaneous(Spawn Count=1, Use Spawn Probability=false)
   &SpawnBurst_Instantaneous(Spawn Group=1, Use Spawn Probability=false, UseLoopCountLimit=false) {module=SpawnBurst_Instantaneous001}
  ]
  particleSpawnScript [
   /SetVariables(Particles.SpawnGroup=$Emitter.SpawnGroup) {isSetParameters=true}
   &InitializeParticle(
    Color=&MergeHelper_ColorSwitch(0=<rgba 0.021 0.0139 0.004 3>, 1=<rgba 1 0.6467 0.4271 5>, 2=?, 3=?, Index=$Particles.SpawnGroup, Num=::NewEnumerator0)
    Lifetime=&MergeHelper_FloatSwitch(0=0.22, 1=0.4, 2=?, 3=?, Index=$Particles.SpawnGroup, Num=::NewEnumerator0)
    Sprite Size Mode=::NewEnumerator3
    Uniform Sprite Size=200
    UsePositionOffset=false
    Write Material Random=true
   ) {module=InitializeParticle001}
   &DynamicMaterialParameters(
    Index 0 Param 1=1
    Index 0 Param 2=1
    Index 0 Param 3=1
    Index 0 Param 4=1
    Module.Param0WriteEnabled=true
    Module.Param1WriteEnabled=true
    Module.Param2WriteEnabled=true
    Module.Param3WriteEnabled=true
   )
   /SetVariables(Particles.VisibilityTag=$Particles.SpawnGroup) {isSetParameters=true}
  ]
  particleUpdateScript [
   &CameraOffset(Camera Offset Amount=20)
   &ParticleState()
   &SolveForcesAndVelocity(Clamp Velocity=false, Limit Acceleration=false)
   &ScaleColor(
    Scale Alpha=&Multiply_Float(
     A=&FloatFromCurve(
      FloatCurve=DataInterface() {propertyValues {Curve {keys [<at 0 1>,<at 1 0>]}}}
      Scale Curve=&FN_FrontEndFloatModifier(Gameplay=1, Front End=0.5)
     )
     B=&FNTimeOfDayScalar(Night Scale=0.3, Morning Scale=0.05, Day Scale=0.5, Evening Scale=0.05)
    )
    ScaleA=true
    ScaleRGB=true
   )
   &ScaleSpriteSize(Scale Factor=&MergeHelper_Vector2Switch(
     0=&Vector2DFromFloat(Value=&FloatFromCurve(
       FloatCurve=DataInterface() {propertyValues {Curve {keys [<at 0 1>,<at 1 1.15>]}}}
       Scale Curve=&FN_FrontEndFloatModifier(Gameplay=1, Front End=0.5)
      ))
     1=&Vector2DFromFloat(Value=&FloatFromCurve(
       FloatCurve=DataInterface() {propertyValues {Curve {keys [{arriveTangent=0.8072, interpMode=RCIM_Cubic, leaveTangent=0.8072, tangentMode=RCTM_User, time=0, value=0.7}, <at 1 1.075>]}}}
       Scale Curve=&FN_FrontEndFloatModifier(Gameplay=1, Front End=0.5)
      ))
     2=?
     3=?
     4=?
     Index=$Particles.SpawnGroup
     Num=::NewEnumerator0
    ))
   /SetVariables(Particles.CoronaColor:LinearColor=&ColorFromCurve(ColorCurve=DataInterface() {propertyValues {
       AlphaCurve {keys [<at 0 1.1 RCIM_Cubic>,<at 1 0 RCIM_Cubic>]}
       BlueCurve {keys [<at 0 0 RCIM_Cubic>]}
       GreenCurve {keys [<at 0 0.822 RCIM_Cubic>]}
       RedCurve {keys [<at 0 3 RCIM_Cubic>]}
      }})) {isSetParameters=true}
  ]
  rendererClasses [&NiagaraSpriteRendererProperties]
  renderers [
   {rendererClass=&NiagaraSpriteRendererProperties}
   {rendererClass=&NiagaraSpriteRendererProperties, rendererIndex=1}
  ]
 }
 Flash_Flare {
  emitterUpdateScript [
   &EmitterState()
   &SpawnBurst_Instantaneous(Spawn Count=1, Use Spawn Probability=false)
  ]
  particleSpawnScript [
   &InitializeParticle(
    Color=<rgba 1.5 1.5 1.5 1>
    Lifetime=0.25
    Sprite Size Mode=::NewEnumerator3
    Uniform Sprite Size=180
    UsePositionOffset=false
    Write Material Random=true
   ) {module=InitializeParticle001}
   &DynamicMaterialParameters(
    Index 0 Param 1=1
    Index 0 Param 2=1
    Module.Param0WriteEnabled=true
    Module.Param1WriteEnabled=true
    Module.Param2WriteEnabled=true
    Module.Param3WriteEnabled=true
   )
   &CameraOffset(Camera Offset Amount=5)
  ]
  particleUpdateScript [
   &ParticleState()
   &ScaleSpriteSize(
    Uniform Curve Scale=&FN_FrontEndFloatModifier(Gameplay=1, Front End=$System.FrontendScale)
    Uniform Curve Sprite Scale=DataInterface() {propertyValues {Curve {keys [
        {arriveTangent=2.9559, interpMode=RCIM_Cubic, leaveTangent=2.9559, tangentMode=RCTM_User, time=0, value=0.6}
        {arriveTangent=0.0224, interpMode=RCIM_Cubic, leaveTangent=0.0224, tangentMode=RCTM_User, time=0.3, value=1}
        <at 1 0>
       ]}}}
   )
   &ScaleColor(
    Scale Alpha=&FNTimeOfDayScalar(Night Scale=0.75, Morning Scale=1, Evening Scale=1)
    Scale RGB=&MakeVectorFromLinearColorRGB(Color=&ColorFromCurve(ColorCurve=DataInterface() {propertyValues {
        AlphaCurve {keys [<at 0 0>,<at 0.1287 1 RCIM_Cubic>,<at 1 0>]}
        BlueCurve {keys [<at 0.004 0>, {arriveTangent=0.077, interpMode=RCIM_Cubic, leaveTangent=0.077, time=0.1965, value=0}, <at 1 0.0767 RCIM_Cubic>]}
        GreenCurve {keys [<at 0.004 0.3994>, {arriveTangent=0.0462, interpMode=RCIM_Cubic, leaveTangent=0.0462, time=0.1965, value=0.8274}, <at 1 0.4454 RCIM_Cubic>]}
        RedCurve {keys [<at 0.004 1>,<at 0.1965 2 RCIM_Cubic>,<at 1 1 RCIM_Cubic>]}
       }}))
    ScaleA=true
    ScaleRGB=true
   )
   &FN_Sprite_Attributes(
    Color=&ColorFromCurve(ColorCurve=DataInterface() {propertyValues {
       AlphaCurve {keys [<at 0 3 RCIM_Cubic>,<at 1 0 RCIM_Cubic>]}
       BlueCurve {keys [<at 0 0 RCIM_Cubic>,<at 0.9989 0>]}
       GreenCurve {keys [<at 0 0 RCIM_Cubic>,<at 0.9989 0>]}
       RedCurve {keys [<at 0 0 RCIM_Cubic>,<at 0.9989 0>]}
      }})
    Create Color Attribute=true
    Create SpriteSize Attribute=true
    SpriteSize=&Vector2DFromFloat(Value=&FloatFromCurve(
      FloatCurve=DataInterface() {propertyValues {Curve {keys [
          {arriveTangent=7.1084, interpMode=RCIM_Cubic, leaveTangent=7.1084, tangentMode=RCTM_User, time=0, value=0.375}
          {arriveTangent=-2.821, interpMode=RCIM_Cubic, leaveTangent=-2.821, tangentMode=RCTM_User, time=0.5, value=0.625}
          {tangentMode=RCTM_User, time=1, value=0}
         ]}}}
      Scale Curve=&Multiply_Float(A=120, B=&FN_FrontEndFloatModifier(Gameplay=1, Front End=$System.FrontendScale))
     ))
    Write to Module Namespace=true
   )
  ]
  rendererClasses [&NiagaraSpriteRendererProperties]
  renderers [
   {rendererClass=&NiagaraSpriteRendererProperties}
   {rendererClass=&NiagaraSpriteRendererProperties, rendererIndex=1}
  ]
 }
 MeshShapes {
  emitterUpdateScript [
   &EmitterState()
   &SpawnBurst_Instantaneous(
    Spawn Count=&RandomRangeInt(Minimum=4, Maximum=7)
    Spawn Group=1
    Use Spawn Probability=false
    UseLoopCountLimit=false
   )
  ]
  particleSpawnScript [
   &InitializeParticle(
    Color=&SelectLinearColorFromArray(
     Array Sampling Mode=::NewEnumerator0
     Direct Array Index=?
     Color Selection Array=DataInterface() {propertyValues {ColorData [<rgba 1.5 0.6429 0.1537 1>, <rgba 1.5 0.6429 0.1537 1>]}}
     Random Seed=0
    )
    Lifetime Max=0.4
    Lifetime Min=0.25
    Lifetime Mode=::NewEnumerator1
    Mesh Renderer Array Visibility Mode=::NewEnumerator1
    Mesh Renderer Info=DataInterface() {propertyValues {MeshRenderer=&self:MeshShapes_3.NiagaraMeshRendererProperties_2}}
    Mesh Scale Mode=::NewEnumerator3
    Mesh Uniform Scale=&MakeCustomFloatFromBool(
     Bool=&Boolean_OrOperation(Bool1=$User.bIsFrontEnd, Bool2=$User.bIsFrontEndPreview)
     True Float=&RandomRangeFloat(Minimum=0.3, Maximum=0.6)
     False Float=&RandomRangeFloat(Minimum=0.6, Maximum=0.9)
    )
    Position Offset Coordinate Space=::Local
    Sprite UV Mode=::NewEnumerator3
    UsePositionOffset=true
    Write Material Random=true
   ) {module=InitializeParticle001}
   &ShapeLocation(
    Hemisphere Distribution=<1 0.5>
    MaskSpawnGroup=false
    Override Seed=false
    Sphere Radius=&FN_FrontEndFloatModifier(Gameplay=15, Front End=8)
    Sphere Surface Distribution=0.5
    UseOwnerScale=true
   )
   &InitialMeshOrientation(Mesh Orientation Mode=::NewEnumerator1, Rotation=<0.75 0 0>, Use Rotation Vector=true) {module=InitialMeshOrientation001}
   &AddVelocity(
    Velocity=&RandomRangeVector(Minimum=<200 -1300 -1300>, Maximum=<600 1300 1300>)
    Velocity Speed Scale=&MakeCustomFloatFromBool(
     Bool=&Boolean_OrOperation(Bool1=$User.bIsFrontEnd, Bool2=$User.bIsFrontEndPreview)
     True Float=&RandomRangeFloat(Minimum=0.1, Maximum=0.1)
     False Float=&RandomRangeFloat(Minimum=0.3, Maximum=0.8)
    )
   ) {module=AddVelocity001}
   &MeshRotationForce(Rotation=&VectorFromFloat(Value=&RandomRangeFloat(Minimum=-3, Maximum=3)))
   &ApplyInitialForces()
   &CameraOffset(Camera Offset Amount=&FN_FrontEndFloatModifier(Gameplay=25, Front End=0))
   /SetVariables(Particles.MeshIndex=&RandomRangeInt(Minimum=0, Maximum=1)) {isSetParameters=true}
   &DynamicMaterialParameters(
    Index 0 Param 1=&MergeHelper_FloatSwitch(0=0, 1=0.5, 2=?, 3=?, Index=$Particles.MeshIndex, Num=::NewEnumerator0)
    Index 0 Param 2=2
    Module.Index0Param0Write=true
    Module.Index0Param1Write=true
    Module.Index0Param2Write=true
    Module.Index0Param3Write=true
   )
   &SubUVAnimation(
    Mesh Renderer=DataInterface() {propertyValues {MeshRenderer=&self:MeshShapes_3.NiagaraMeshRendererProperties_2}}
    Output Alternate Renderer Binding=true
    Renderer Type:ENiagaraMeshOrSprite=::NewEnumerator1
    UseEndFrame=false
    UseStartFrame=false
   ) {enabled=false}
  ]
  particleUpdateScript [
   &ParticleState()
   &SpriteRotationRate() {enabled=false, module=SpriteRotationRate001}
   &ScaleMeshSize(Scale Factor=&Multiply_VectorByFloat(
     Vector=<1 1 1>
     Float=&FloatFromCurve(FloatCurve=DataInterface() {propertyValues {Curve {keys [
          {interpMode=RCIM_Cubic, tangentMode=RCTM_User, time=0, value=0}
          <at 0.2 1>
          {interpMode=RCIM_Cubic, tangentMode=RCTM_User, time=0.5, value=1}
          {arriveTangent=-0.9819, interpMode=RCIM_Cubic, leaveTangent=-0.9819, tangentMode=RCTM_User, time=0.75, value=1.3}
          {arriveTangent=-9.5654, interpMode=RCIM_Cubic, leaveTangent=-9.5654, tangentMode=RCTM_User, time=1, value=0}
         ]}}})
    )) {module=ScaleMeshSize001}
   &Drag(Drag=&RandomRangeFloat(Minimum=8, Maximum=12), Use Linear Drag=true, Use Rotational Drag=false) {module=Drag001}
   &SpriteRotationRate(Rotation Rate=&RandomRangeFloat(Minimum=40, Maximum=-50))
   &GravityForce(Gravity=&RandomRangeVector(Minimum=<0 0 150>, Maximum=<0 0 -150>))
   &CurlNoiseForce(Pan Noise=false)
   &SolveForcesAndVelocity(Clamp Velocity=false, Limit Acceleration=false)
   &ScaleColor(
    Scale RGB=&VectorFromFloat(Value=&Multiply_Float(
      A=&FloatFromCurve(
       FloatCurve=DataInterface() {propertyValues {Curve {keys [
           {arriveTangent=9.0725, interpMode=RCIM_Cubic, leaveTangent=9.0725, tangentMode=RCTM_User, time=0, value=1}
           {arriveTangent=-2.5249, interpMode=RCIM_Cubic, leaveTangent=-2.5249, tangentMode=RCTM_User, time=0.1408, value=2}
           {arriveTangent=0.348, interpMode=RCIM_Cubic, leaveTangent=0.348, time=0.2, value=1.125}
           {arriveTangent=-9.5767, interpMode=RCIM_Cubic, leaveTangent=-9.5767, tangentMode=RCTM_User, time=0.5, value=2.125}
           {arriveTangent=-0.0405, interpMode=RCIM_Cubic, leaveTangent=-0.0405, tangentMode=RCTM_User, time=1, value=1}
          ]}}}
       Scale Curve=0.5
      )
      B=&FNTimeOfDayScalar(Night Scale=0.5, Morning Scale=0.5, Day Scale=0.8, Evening Scale=0.5)
     ))
    ScaleA=true
    ScaleRGB=true
   )
   &FN_Sprite_Attributes(
    Create Position Attribute=true
    Create Sprite Rotation Attribute=true
    Create SpriteSize Attribute=true
    Position:NiagaraPosition=&AddVectorToPosition(
     Position=$Particles.Position
     Vector=&Multiply_VectorByFloat(Vector=&RandomRangeVector(Minimum=<-1 -1 -1>, Maximum=<1 1 1>), Float=5)
    )
    SpriteRotation:NiagaraFloat=&RandomRangeFloat(Minimum=360, Maximum=45)
    SpriteSize=&Multiply_Vector2D_ByFloat(
     Float=&FloatFromCurve(
      FloatCurve=DataInterface() {propertyValues {Curve {keys [
          <at 0 0 RCIM_Cubic>
          {arriveTangent=2.4583, interpMode=RCIM_Cubic, leaveTangent=2.4583, time=0.3, value=0.2}
          {arriveTangent=-0.2857, interpMode=RCIM_Cubic, leaveTangent=-0.2857, time=0.6, value=1.475}
          <at 1 0 RCIM_Cubic>
         ]}}}
      Scale Curve=&FN_FrontEndFloatModifier(Gameplay=3, Front End=1)
     )
     Vector2D=<10 10>
    )
   )
   /SetVariables(
    Particles.SparkleColor:LinearColor=&ColorFromCurve(ColorCurve=DataInterface() {propertyValues {
       AlphaCurve {keys [<at 0 0.5 RCIM_Cubic>,<at 1 0 RCIM_Cubic>]}
       BlueCurve {keys [<at 0.5628 0 RCIM_Cubic>,<at 0.8538 1 RCIM_Cubic>]}
       GreenCurve {keys [<at 0.5628 3.9301 RCIM_Cubic>,<at 0.8538 1 RCIM_Cubic>]}
       RedCurve {keys [<at 0.5628 8 RCIM_Cubic>,<at 0.8538 1 RCIM_Cubic>]}
      }})
    Particles.SparkleOffset=10
   ) {isSetParameters=true}
   /SetVariables() {enabled=false, isSetParameters=true}
   &CameraOffset() {enabled=false, module=CameraOffset001}
   &DynamicMaterialParameters(
    Index 1 Param 1=&RandomRangeFloat(Minimum=0, Maximum=0)
    Index 1 Param 2=&FloatFromCurve(FloatCurve=DataInterface() {propertyValues {Curve {keys [
         {arriveTangent=4.1785, interpMode=RCIM_Cubic, leaveTangent=4.1785, tangentMode=RCTM_User, time=0, value=-2}
         {arriveTangent=0.0224, interpMode=RCIM_Cubic, leaveTangent=0.0224, tangentMode=RCTM_User, time=1, value=0.75}
        ]}}})
    Index 1 Param 3=&FloatFromCurve(
     FloatCurve=DataInterface() {propertyValues {Curve {keys [
         {arriveTangent=9.0725, interpMode=RCIM_Cubic, leaveTangent=9.0725, tangentMode=RCTM_User, time=0, value=0}
         {arriveTangent=-2.5249, interpMode=RCIM_Cubic, leaveTangent=-2.5249, tangentMode=RCTM_User, time=0.1408, value=0.9847}
         {arriveTangent=-1.146, interpMode=RCIM_Cubic, leaveTangent=-1.146, time=0.4, value=1}
         {arriveTangent=-0.0405, interpMode=RCIM_Cubic, leaveTangent=-0.0405, tangentMode=RCTM_User, time=1, value=0}
        ]}}}
     Scale Curve=&FNTimeOfDayScalar(Night Scale=1, Morning Scale=1, Day Scale=2, Evening Scale=1)
    )
    Module.Index1Param0Write=true
    Module.Index1Param1Write=true
    Module.Index1Param2Write=true
    Module.Index1Param3Write=true
    Write Parameter Index 0=false
    Write Parameter Index 1=true
   ) {module=DynamicMaterialParameters001}
  ]
  rendererClasses [&NiagaraMeshRendererProperties, &NiagaraSpriteRendererProperties]
  renderers [
   {rendererClass=&NiagaraMeshRendererProperties, rendererIndex=0}
   {rendererClass=&NiagaraSpriteRendererProperties, rendererIndex=1}
  ]
 }
 Sparks_Toon {
  emitterUpdateScript [
   &EmitterState(Life Cycle Mode=::NewEnumerator1, Loop Behavior=::NewEnumerator1, Loop Delay=0.005, Loop Duration=0.25, UseLoopDelay=false)
   &SpawnRate(SpawnRate=20, Use Spawn Probability=false)
   &SpawnBurst_Instantaneous(Spawn Count=20, Use Spawn Probability=false, UseLoopCountLimit=false)
   &SpawnBurst_Instantaneous(
    Spawn Count=&RandomRangeInt(Minimum=12, Maximum=18)
    Spawn Group=1
    Use Spawn Probability=false
    UseLoopCountLimit=false
   ) {module=SpawnBurst_Instantaneous001}
   &SpawnParticlesFromOtherEmitter(
    Attribute Reader=DataInterface() {propertyValues {EmitterBinding {bindingMode=Other, emitterName=MeshShapes}}}
    Spawn Group=2
    Spawn Rate=15
    Use Maximum Age=false
    Use Minimum Age=false
   )
  ]
  particleSpawnScript [
   &FN_SetSpawnGroup(Number of Spawn Groups=2) {enabled=false}
   /SetVariables(Particles.SpawnGroup=$Emitter.SpawnGroup) {isSetParameters=true}
   &InitializeParticle(
    Color=<rgba 0.05 0.05 0.05 0.5>
    Lifetime Max=0.6
    Lifetime Min=0.4
    Lifetime Mode=::NewEnumerator1
    Sprite Size Mode=::NewEnumerator1
    Uniform Sprite Size Max=15
    Uniform Sprite Size Min=10
    UsePositionOffset=false
    Write Material Random=true
   )
   &SampleParticlesFromOtherEmitter(Position Sampling:ENiagara_DirectReadSamplingMode=::NewEnumerator1, Velocity Sampling=::NewEnumerator1)
   /SetVariables(Particles.Position=&MergeHelper_VectorSwitch(
     0=Linked() {linkedVariable {name=StackContext.Position, type=NiagaraPosition}}
     1=Linked() {linkedVariable {name=StackContext.Position, type=NiagaraPosition}}
     2=$Output.SampleParticlesFromOtherEmitter.SampledPosition
     3=?
     Index=$Particles.SpawnGroup
     Num=::NewEnumerator1
    )) {isSetParameters=true}
   &SphereLocation(
    MaskSpawnGroup=false
    Override Seed=false
    Sphere Radius=&Multiply_Float(
     A=&RandomRangeFloat(Minimum=2, Maximum=25)
     B=&FN_FrontEndFloatModifier(Gameplay=1, Front End=1)
    )
    Surface Expansion Mode=::NewEnumerator1
    Surface Only=true
    UseNonUniformScale=false
    UseOffset=true
    UseSphereOrientation=false
   )
   &CameraOffset(Camera Offset Amount=1)
   /SetVariables(Particles.VisibilityTag=$Particles.SpawnGroup) {isSetParameters=true}
   &AddVelocity(
    Module.Use Falloff Distance=false
    Override Seed=false
    Velocity Mode=::NewEnumerator1
    Velocity Speed=&Multiply_Float(
     A=&RandomRangeFloat(Minimum=10, Maximum=400)
     B=&RandomRangeFloat(Minimum=10, Maximum=400)
    )
   ) {enabled=false, module=AddVelocity001}
   &AddVelocity(
    Module.Use Falloff Distance=false
    Override Seed=false
    Velocity Mode=::NewEnumerator1
    Velocity Speed=&MergeHelper_FloatSwitch(
     0=0
     1=&Multiply_Float(
      A=&RandomRangeFloat(Minimum=10, Maximum=400)
      B=&FN_FrontEndFloatModifier(Gameplay=0.584, Front End=$System.FrontendScale)
     )
     2=?
     3=?
     Index=$Particles.SpawnGroup
     Num=::NewEnumerator0
    )
   )
   /SetVariables(Particles.SpriteSize=&Multiply_Vector2D_ByFloat(
     Float=&MergeHelper_FloatSwitch(0=1, 1=0.5, 2=?, 3=?, Index=$Particles.SpawnGroup, Num=::NewEnumerator0)
     Vector2D=$StackContext.SpriteSize
    )) {isSetParameters=true}
   &AddVelocity(
    Velocity=&MergeHelper_VectorSwitch(
     0=$Output.SphereLocation.SphereNormal
     1=$Output.SphereLocation.SphereNormal
     2=$Output.SphereLocation.SphereNormal
     3=?
     Index=$Particles.SpawnGroup
     Num=::NewEnumerator1
    )
    Velocity Speed Scale=&MergeHelper_FloatSwitch(
     0=&RandomRangeFloat(Minimum=1, Maximum=50)
     1=1
     2=33
     3=?
     Index=$Particles.SpawnGroup
     Num=::NewEnumerator1
    )
   ) {module=AddVelocity002}
  ]
  particleUpdateScript [
   &ParticleState()
   &SolveForcesAndVelocity(Clamp Velocity=false, Limit Acceleration=false)
   &Color(
    Color=&MergeHelper_ColorSwitch(
     0=&ColorFromCurve(ColorCurve=DataInterface() {propertyValues {
        AlphaCurve {keys [<at 0.0065 0>,<at 0.3936 0.5 RCIM_Cubic>,<at 1 0 RCIM_Cubic>]}
        BlueCurve {keys [<at 0 0 RCIM_Cubic>,<at 0.7105 0 RCIM_Cubic>]}
        GreenCurve {keys [<at 0 0.75 RCIM_Cubic>,<at 0.7105 0.0368 RCIM_Cubic>]}
        RedCurve {keys [<at 0 3 RCIM_Cubic>,<at 0.7105 0.3594 RCIM_Cubic>]}
       }})
     1=<rgba 0.1 0.0495 0.015 1>
     2=?
     3=?
     Index=$Particles.SpawnGroup
     Num=::NewEnumerator0
    )
    Scale Alpha=&FloatFromCurve(
     FloatCurve=DataInterface() {propertyValues {Curve {keys [
         {interpMode=RCIM_Cubic, tangentMode=RCTM_User, time=-0.0063, value=0.0184}
         {arriveTangent=-0.8483, interpMode=RCIM_Cubic, leaveTangent=-0.8483, tangentMode=RCTM_User, time=0.1718, value=1.7147}
         {arriveTangent=-5.0978, interpMode=RCIM_Cubic, leaveTangent=-5.0978, tangentMode=RCTM_User, time=0.2818, value=1.2143}
         {arriveTangent=-0.2783, interpMode=RCIM_Cubic, leaveTangent=-0.2783, tangentMode=RCTM_User, time=0.4, value=0.7}
         {arriveTangent=0.3112, interpMode=RCIM_Cubic, leaveTangent=0.3112, tangentMode=RCTM_User, time=0.5407, value=1.3419}
         {arriveTangent=-3.7119, interpMode=RCIM_Cubic, leaveTangent=-3.7119, tangentMode=RCTM_User, time=0.7242, value=0.3749}
         {interpMode=RCIM_Cubic, tangentMode=RCTM_User, time=1.0011, value=0.002}
        ]}}}
     Scale Curve=0.8
    )
    Scale Color=&VectorFromFloat(Value=50)
   )
   &ScaleSpriteSize(Scale Factor=&Vector2DFromFloat(Value=&FloatFromCurve(
      FloatCurve=DataInterface() {propertyValues {Curve {keys [
          {arriveTangent=4.1302, interpMode=RCIM_Cubic, leaveTangent=4.1302, tangentMode=RCTM_User, time=0, value=0}
          {interpMode=RCIM_Cubic, tangentMode=RCTM_User, time=0.5, value=1}
          {tangentMode=RCTM_User, time=1, value=0}
         ]}}}
      Scale Curve=&Multiply_Float(
       A=&MaskFloatBySpawnGroup(Masked Float=1, Passthrough Float=1.5)
       B=&FN_FrontEndFloatModifier(Gameplay=1, Front End=0.5)
      )
     )))
   &Drag(
    Drag=&MergeHelper_FloatSwitch(0=0, 1=6, 2=?, 3=?, Index=$Particles.SpawnGroup, Num=::NewEnumerator0)
    Use Linear Drag=true
    Use Rotational Drag=true
   )
   &SolveForcesAndVelocity(Clamp Velocity=false, Limit Acceleration=false) {module=SolveForcesAndVelocity001}
  ]
  rendererClasses [&NiagaraSpriteRendererProperties]
  renderers [
   {rendererClass=&NiagaraSpriteRendererProperties}
   {rendererClass=&NiagaraSpriteRendererProperties, rendererIndex=1}
   {rendererClass=&NiagaraSpriteRendererProperties, rendererIndex=2}
  ]
 }
 SpeedLine_PACKED_TWOCOLOR {
  emitterUpdateScript [
   &EmitterState()
   &SpawnBurst_Instantaneous(Spawn Count=1, Use Spawn Probability=false)
  ]
  particleSpawnScript [
   &InitializeParticle(
    Color=<rgba 2 0.7695 0 8>
    Lifetime=0.2
    Sprite Rotation Mode=::NewEnumerator2
    Sprite Size Mode=::NewEnumerator3
    Uniform Sprite Size=&Multiply_Float(
     A=&RandomRangeFloat(Minimum=110, Maximum=130)
     B=&FN_FrontEndFloatModifier(Gameplay=1, Front End=&Multiply_Float(A=$User.FrontEndScale, B=0.85))
    )
    UsePositionOffset=false
    Write Material Random=true
   ) {module=InitializeParticle001}
   &FN_Sprite_Attributes(
    "Create Normalized Age "=true
    Create Sprite Rotation Attribute=true
    Create SpriteSize Attribute=true
    Lifetime=0.2
    Max Age=0.15
    SpriteRotation:NiagaraFloat=&RandomRangeFloat(Minimum=-360, Maximum=360)
    SpriteSize=&Vector2DFromFloat(Value=&Multiply_Float(
      A=&RandomRangeFloat(Minimum=110, Maximum=130)
      B=&FN_FrontEndFloatModifier(Gameplay=1, Front End=&Multiply_Float(A=$User.FrontEndScale, B=0.7))
     ))
    Write to Module Namespace=true
   )
  ]
  particleUpdateScript [
   &ParticleState()
   &ScaleSpriteSize(Scale Factor=&Vector2DFromFloat(Value=&FloatFromCurve(FloatCurve=DataInterface() {propertyValues {Curve {keys [{arriveTangent=1.3574, interpMode=RCIM_Cubic, leaveTangent=1.3574, tangentMode=RCTM_User, time=0, value=0.6}, <at 1 1.4 RCIM_Cubic>]}}})))
   /SetVariables(Particles.SpriteSizeSecond:Vector2f=&Multiply_Vector2D(
     A=$Particles.Initial.FN_Sprite_Attributes.SpriteSize
     B=&Vector2DFromFloat(Value=&FloatFromCurve(FloatCurve=DataInterface() {propertyValues {Curve {keys [{arriveTangent=1.3574, interpMode=RCIM_Cubic, leaveTangent=1.3574, tangentMode=RCTM_User, time=0, value=0.8}, <at 1 1.4 RCIM_Cubic>]}}}))
    )) {isSetParameters=true}
   &DynamicMaterialParameters(
    Index 0 Param 1=&FloatFromCurve(FloatCurve=DataInterface() {propertyValues {Curve {keys [<at 0 1 RCIM_Cubic>]}}})
    Index 0 Param 2=&FloatFromCurve(FloatCurve=DataInterface() {propertyValues {Curve {keys [<at 0 0 RCIM_Cubic>, {arriveTangent=2.4833, interpMode=RCIM_Cubic, leaveTangent=2.4833, tangentMode=RCTM_User, time=1, value=1}]}}})
    Module.Param0WriteEnabled=true
    Module.Param1WriteEnabled=true
    Module.Param2WriteEnabled=true
    Module.Param3WriteEnabled=true
   )
   &ScaleColor(
    Scale Alpha=&FloatFromCurve(FloatCurve=DataInterface() {propertyValues {Curve {keys [
         {arriveTangent=-0.2921, interpMode=RCIM_Cubic, leaveTangent=-0.2921, tangentMode=RCTM_User, time=0, value=7}
         {arriveTangent=-7, interpMode=RCIM_Cubic, leaveTangent=-7, time=0.4, value=4}
         <at 1 0 RCIM_Cubic>
        ]}}})
    Scale RGB=&Multiply_VectorByFloat(
     Vector=<1 1 1>
     Float=&FNTimeOfDayScalar(Night Scale=0.75, Morning Scale=0.9, Evening Scale=0.9)
    )
    ScaleA=true
    ScaleRGB=true
   )
   &CameraOffset(Camera Offset Amount=&FN_FrontEndFloatModifier(Gameplay=25, Front End=5))
  ]
  rendererClasses [&NiagaraSpriteRendererProperties]
  renderers [{rendererClass=&NiagaraSpriteRendererProperties}]
 }
 Toon_Smoke {
  emitterUpdateScript [
   &EmitterState()
   &SpawnBurst_Instantaneous(Spawn Count=1, Spawn Time=0.03, Use Spawn Probability=false)
  ]
  particleSpawnScript [
   /SetVariables(Particles.Random=&RandomRangeFloat(Minimum=0, Maximum=1)) {isSetParameters=true}
   &InitializeParticle(
    Color=&FNTimeOfDayColor(
     Night Color=<rgba 0.3 0.2805 0.21 1>
     Morning Color=<rgba 0.45 0.4207 0.315 1>
     Day Color=<rgba 1 1 1 1>
     Evening Color=<rgba 0.5 0.3585 0.3025 1>
    )
    Lifetime=&Lerp_Float(A=0.3, B=0.35, Alpha=$Particles.Random)
    Mass Mode=::NewEnumerator1
    Sprite Rotation Angle=&Lerp_Float(A=-45, B=45, Alpha=$Particles.Random)
    Sprite Rotation Mode=::NewEnumerator0
    Sprite Size=&Multiply_Vector2D_ByFloat(
     Float=&FN_FrontEndFloatModifier(Gameplay=1, Front End=0.65)
     Vector2D=&Vector2DFromFloat(Value=&Lerp_Float(A=120, B=140, Alpha=$Particles.Random))
    )
    Sprite Size Mode=::NewEnumerator4
    UsePositionOffset=false
    Write Material Random=true
   ) {module=InitializeParticle001}
   &CameraOffset(Camera Offset Amount=&FN_FrontEndFloatModifier(Gameplay=30, Front End=5))
   &DynamicMaterialParameters(
    Index 0 Param 2=&UniformAOrBFloat(A=0)
    Module.Param0WriteEnabled=true
    Module.Param1WriteEnabled=true
    Module.Param2WriteEnabled=true
    Module.Param3WriteEnabled=true
   ) {module=DynamicMaterialParameters001}
  ]
  particleUpdateScript [
   &ParticleState()
   &DynamicMaterialParameters(
    Index 0 Param 1=&FloatFromCurve(FloatCurve=DataInterface() {propertyValues {Curve {keys [<at 0 0>,<at 1 1>]}}})
    Index 0 Param 2=1
    Module.Param0WriteEnabled=true
    Module.Param1WriteEnabled=true
    Module.Param2WriteEnabled=true
    Module.Param3WriteEnabled=true
   )
   &ScaleColor(
    Scale Alpha=&FloatFromCurve(FloatCurve=DataInterface() {propertyValues {Curve {keys [<at 0 1>,<at 0.8091 0.9932>,<at 1 0>]}}})
    Scale RGB=&VectorFromFloat(Value=$System.TOD)
    ScaleA=true
    ScaleRGB=true
   )
   &Color(
    Color=&ColorFromCurve(ColorCurve=DataInterface() {propertyValues {
       AlphaCurve {keys [<at 0 1 RCIM_Cubic>,<at 0.7039 0 RCIM_Cubic>]}
       BlueCurve {keys [<at 0.166 0 RCIM_Cubic>,<at 1 0 RCIM_Cubic>]}
       GreenCurve {keys [<at 0.166 8.8801 RCIM_Cubic>,<at 1 3.0187 RCIM_Cubic>]}
       RedCurve {keys [<at 0.166 32 RCIM_Cubic>,<at 1 5 RCIM_Cubic>]}
      }})
    Scale Alpha=0.2
   )
  ]
  rendererClasses [&NiagaraSpriteRendererProperties]
  renderers [{rendererClass=&NiagaraSpriteRendererProperties}]
 }
}
summary {
 systemName=NS_Pickaxe_DollEnthusiast_Gold_Impact
 emitters [
  {emitterName=SpeedLine_PACKED_TWOCOLOR, rendererClasses [&NiagaraSpriteRendererProperties]}
  {emitterName=MeshShapes, rendererClasses [&NiagaraMeshRendererProperties, &NiagaraSpriteRendererProperties]}
  {emitterName=Toon_Smoke, rendererClasses [&NiagaraSpriteRendererProperties]}
  {emitterName=Dark_Glow, rendererClasses [&NiagaraSpriteRendererProperties]}
  {emitterName=Sparks_Toon, rendererClasses [&NiagaraSpriteRendererProperties]}
  {emitterName=CoreFlashImpact001, rendererClasses [&NiagaraSpriteRendererProperties]}
  {emitterName=Flash_Flare, rendererClasses [&NiagaraSpriteRendererProperties]}
 ]
 userVariables [
  User.bIsFrontEnd {type=NiagaraBool}
  User.bIsFrontEndPreview {type=NiagaraBool}
  User.FrontEndScale {type=NiagaraFloat, defaultValue=0.558, description=""}
  User.VariantIndex {type=NiagaraInt32}
  User.VariantPicker {type=NiagaraFloat, defaultValue=0, description=""}
 ]
}
